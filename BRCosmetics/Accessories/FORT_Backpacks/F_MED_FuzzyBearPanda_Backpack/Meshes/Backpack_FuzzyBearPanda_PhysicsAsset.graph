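# Physics asset for the Fuzzy Bear Panda backpack cosmetic, defining collision and joint limits for its skeletal mesh.
# A box body sits on the backpack root, with capsule bodies on three dynamic branch bones: front, left and right.
# Each branch is constrained directly to the root with matching limited swing cones and a narrower twist range, so the branches can dangle independently within bounds.
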PhysicsAsset /BRCosmetics/Accessories/FORT_Backpacks/F_MED_FuzzyBearPanda_Backpack/Meshes/Backpack_FuzzyBearPanda_PhysicsAsset
bodies {
 backpack_root [{shapeType=Box, extentX=4.8332, extentY=4.7885, extentZ=5.7231}]
 dyn_branch_front [{shapeType=Capsule, length=8.9224, radius=6.3807, center=<8.0969 0.1882 -4.5353>, rotation=<rot 12.5153 -82.6112 -52.4212>}]
 dyn_branch_left [{shapeType=Capsule, length=5.1387, radius=3.4166, center=<5.3268 -0.1395 -1.568>, rotation=<rot 7.4701 -92.8929 -76.8719>}]
 dyn_branch_right [{shapeType=Capsule, length=8.9746, radius=7.1961, center=<8.3264 1.4107 -3.2064>, rotation=<rot 4.2166 -72.9082 -65.2843>}]
}
constraints [
 {bone1Name=dyn_branch_front, bone2Name=backpack_root, swing1LimitDegrees=45, swing2LimitDegrees=45, twistLimitDegrees=20}
 {bone1Name=dyn_branch_right, bone2Name=backpack_root, swing1LimitDegrees=45, swing2LimitDegrees=45, twistLimitDegrees=20}
 {bone1Name=dyn_branch_left, bone2Name=backpack_root, swing1LimitDegrees=45, swing2LimitDegrees=45, twistLimitDegrees=20}
]
